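# Physics asset for the left-arm jacket cuff on the female medium Bullet Blue character body, defining the bodies and joints for cuff secondary motion.
# Front, back, top and bottom dyn_jacketArm bones each carry an identically sized box, constrained to a matching MSTR bone with a tiny box as anchor.
# Tight angular limits allow more swing on one axis, which alternates between the front/back and top/bottom pairs.
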PhysicsAsset /BRCosmetics/Characters/Player/Female/Medium/Bodies/F_MED_BulletBlue/Meshes/F_MED_Bullet_Blue_LarmCuff_Physics
bodies {
 dyn_jacketArmBack_01_l [{shapeType=Box, extentX=4.2985, extentY=2.4953, extentZ=6.1023, center=<2.0799 -0.0001 0.0001>, rotation=<rot 0 0 0.0002>}]
 dyn_jacketArmBack_01_l_MSTR [{shapeType=Box, extentX=0.01, extentY=0.01, extentZ=0.01}]
 dyn_jacketArmBottom_01_l [{shapeType=Box, extentX=4.2985, extentY=2.4953, extentZ=6.1023, center=<2.0799 -0.0001 0.0001>, rotation=<rot 0 0 -90.0001>}]
 dyn_jacketArmBottom_01_l_MSTR [{shapeType=Box, extentX=0.01, extentY=0.01, extentZ=0.01}]
 dyn_jacketArmFront_01_l [{shapeType=Box, extentX=4.2985, extentY=2.4953, extentZ=6.1023, center=<2.0799 -0.0001 0.0001>, rotation=<rot 0 0 0.0002>}]
 dyn_jacketArmFront_01_l_MSTR [{shapeType=Box, extentX=0.01, extentY=0.01, extentZ=0.01}]
 dyn_jacketArmTop_01_l [{shapeType=Box, extentX=4.2985, extentY=2.4953, extentZ=6.1023, center=<2.0799 -0.0001 0.0001>, rotation=<rot 0 0 -90.0001>}]
 dyn_jacketArmTop_01_l_MSTR [{shapeType=Box, extentX=0.01, extentY=0.01, extentZ=0.01}]
}
constraints [
 {bone1Name=dyn_jacketArmFront_01_l, bone2Name=dyn_jacketArmFront_01_l_MSTR, swing1LimitDegrees=15, swing2LimitDegrees=5, twistLimitDegrees=5}
 {bone1Name=dyn_jacketArmTop_01_l, bone2Name=dyn_jacketArmTop_01_l_MSTR, swing1LimitDegrees=5, swing2LimitDegrees=15, twistLimitDegrees=5}
 {bone1Name=dyn_jacketArmBack_01_l, bone2Name=dyn_jacketArmBack_01_l_MSTR, swing1LimitDegrees=15, swing2LimitDegrees=5, twistLimitDegrees=5}
 {
  bone1Name=dyn_jacketArmBottom_01_l
  bone2Name=dyn_jacketArmBottom_01_l_MSTR
  swing1LimitDegrees=5
  swing2LimitDegrees=15
  twistLimitDegrees=5
 }
]
