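# Physics asset for the gun belt on the female medium Bullet Blue character, simulating eight belt segments, a left pouch and two left grenades.
# Each segment has tight angular limits to a tiny MSTR anchor body, which rotates freely against spine_01.
# The pouch swings from segment 1 and both grenades from segment 2, with the grenade pair angularly locked together.
# Pelvis, spine and thigh shapes provide the body volumes.
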PhysicsAsset /BRCosmetics/Characters/Player/Female/Medium/Bodies/F_MED_BulletBlue/Meshes/F_MED_Bullet_Blue_LgunBelt_Physics
bodies {
 dyn_beltGrenade2_01_l [{shapeType=Capsule, length=6.9427, radius=2.973, center=<7.6796 -0.407 0.0007>, rotation=<rot 0 -91.5603 -89.6525>}]
 dyn_beltGrenade_01_l [{shapeType=Capsule, length=6.9427, radius=2.973, center=<7.6796 -0.407 0.0007>, rotation=<rot 0 -91.5603 -89.6525>}]
 dyn_beltPouch_01_l [{shapeType=Box, extentX=6.4905, extentY=4.8435, extentZ=16.8438, center=<6.5924 0 1.7759>, rotation=<rot 0 90.0001 -88.8319>}]
 dyn_beltTop1_01 [{shapeType=Box, extentX=12.2985, extentY=2.9953, extentZ=3.1023, center=<6.0117 0 0.0005>}]
 dyn_beltTop1_01_MSTR [{shapeType=Box, extentX=0.01, extentY=0.01, extentZ=0.01}]
 dyn_beltTop2_01 [{shapeType=Box, extentX=12.2985, extentY=2.9953, extentZ=3.1023, center=<6.0117 0 0.0005>}]
 dyn_beltTop2_01_MSTR [{shapeType=Box, extentX=0.01, extentY=0.01, extentZ=0.01}]
 dyn_beltTop3_01 [{shapeType=Capsule, length=13.5, radius=1.755, center=<4.0037 0 0>, rotation=<rot 0 -89.9999 89.9994>}]
 dyn_beltTop3_01_MSTR [{shapeType=Box, extentX=0.01, extentY=0.01, extentZ=0.01}]
 dyn_beltTop4_01 [{shapeType=Box, extentX=12.2985, extentY=2.9953, extentZ=3.1023, center=<6.0117 0 0.0005>, rotation=<rot 7.7994 0 0>}]
 dyn_beltTop4_01_MSTR [{shapeType=Box, extentX=0.01, extentY=0.01, extentZ=0.01}]
 dyn_beltTop5_01 [{shapeType=Box, extentX=12.2985, extentY=2.9953, extentZ=1.6023, center=<6.0117 0 0.0005>}]
 dyn_beltTop5_01_MSTR [{shapeType=Box, extentX=0.01, extentY=0.01, extentZ=0.01}]
 dyn_beltTop6_01 [{shapeType=Box, extentX=12.2985, extentY=2.9953, extentZ=3.1023, center=<6.0117 0 0.0005>}]
 dyn_beltTop6_01_MSTR [{shapeType=Box, extentX=0.01, extentY=0.01, extentZ=0.01}]
 dyn_beltTop7_01 [{shapeType=Box, extentX=12.2985, extentY=2.9953, extentZ=3.1023, center=<6.0117 0 0.0005>}]
 dyn_beltTop7_01_MSTR [{shapeType=Box, extentX=0.01, extentY=0.01, extentZ=0.01}]
 dyn_beltTop8_01 [{shapeType=Box, extentX=12.2985, extentY=2.9953, extentZ=3.1023, center=<6.0117 0 0.0005>}]
 dyn_beltTop8_01_MSTR [{shapeType=Box, extentX=0.01, extentY=0.01, extentZ=0.01}]
 pelvis [{shapeType=Capsule, length=4.25, radius=6.75, center=<-2.7609 -5.0112 1.0002>}]
 spine_01 [{shapeType=Capsule, length=0.5, radius=0.505, center=<3.8835 0 0>}]
 spine_02 [{shapeType=Capsule, length=5.1912, radius=6.8581, center=<3.8105 0.0003 -0.0264>, rotation=<rot -0.7629 -89.9999 0.3972>}]
 spine_03 [{shapeType=Capsule, length=0.505, radius=0.505, rotation=<rot 0 -90 0>}]
 thigh_l [{shapeType=Capsule, length=23.6062, radius=7.6185, center=<-15.6538 0.0001 0>, rotation=<rot 90 -0.0197 -0.0196>}]
 thigh_r [{shapeType=Capsule, length=23.6062, radius=7.6185, center=<14.6625 -0.0004 -0.0788>, rotation=<rot 86.1442 179.9999 180>}]
}
constraints [
 {bone1Name=dyn_beltTop4_01, bone2Name=dyn_beltTop4_01_MSTR, swing1LimitDegrees=5, swing2LimitDegrees=5, twistLimitDegrees=5}
 {bone1Name=dyn_beltTop1_01, bone2Name=dyn_beltTop1_01_MSTR, swing1LimitDegrees=8, swing2LimitDegrees=15, twistLimitDegrees=8}
 {bone1Name=dyn_beltPouch_01_l, bone2Name=dyn_beltTop1_01, swing1LimitDegrees=45, swing2LimitDegrees=45, twistLimitDegrees=5}
 {bone1Name=dyn_beltTop8_01, bone2Name=dyn_beltTop8_01_MSTR, swing1LimitDegrees=5, swing2LimitDegrees=15, twistLimitDegrees=5}
 {bone1Name=dyn_beltTop3_01, bone2Name=dyn_beltTop3_01_MSTR, swing1LimitDegrees=15, swing2LimitDegrees=15, twistLimitDegrees=5}
 {
  bone1Name=dyn_beltTop3_01_MSTR
  bone2Name=spine_01
  swing1LimitDegrees=45
  swing1Motion=Free
  swing2LimitDegrees=45
  swing2Motion=Free
  twistLimitDegrees=45
  twistMotion=Free
 }
 {
  bone1Name=dyn_beltTop4_01_MSTR
  bone2Name=spine_01
  swing1LimitDegrees=45
  swing1Motion=Free
  swing2LimitDegrees=45
  swing2Motion=Free
  twistLimitDegrees=45
  twistMotion=Free
 }
 {
  bone1Name=dyn_beltTop1_01_MSTR
  bone2Name=spine_01
  swing1LimitDegrees=45
  swing1Motion=Free
  swing2LimitDegrees=45
  swing2Motion=Free
  twistLimitDegrees=45
  twistMotion=Free
 }
 {
  bone1Name=dyn_beltTop8_01_MSTR
  bone2Name=spine_01
  swing1LimitDegrees=45
  swing1Motion=Free
  swing2LimitDegrees=45
  swing2Motion=Free
  twistLimitDegrees=45
  twistMotion=Free
 }
 {
  bone1Name=dyn_beltTop7_01_MSTR
  bone2Name=spine_01
  swing1LimitDegrees=45
  swing1Motion=Free
  swing2LimitDegrees=45
  swing2Motion=Free
  twistLimitDegrees=45
  twistMotion=Free
 }
 {
  bone1Name=dyn_beltTop6_01_MSTR
  bone2Name=spine_01
  swing1LimitDegrees=45
  swing1Motion=Free
  swing2LimitDegrees=45
  swing2Motion=Free
  twistLimitDegrees=45
  twistMotion=Free
 }
 {
  bone1Name=dyn_beltTop5_01_MSTR
  bone2Name=spine_01
  swing1LimitDegrees=45
  swing1Motion=Free
  swing2LimitDegrees=45
  swing2Motion=Free
  twistLimitDegrees=45
  twistMotion=Free
 }
 {bone1Name=dyn_beltTop5_01, bone2Name=dyn_beltTop5_01_MSTR, swing1LimitDegrees=10, swing2LimitDegrees=10, twistLimitDegrees=5}
 {bone1Name=dyn_beltTop6_01, bone2Name=dyn_beltTop6_01_MSTR, swing1LimitDegrees=5, swing2LimitDegrees=15, twistLimitDegrees=5}
 {bone1Name=dyn_beltTop7_01, bone2Name=dyn_beltTop7_01_MSTR, swing1LimitDegrees=5, swing2LimitDegrees=15, twistLimitDegrees=5}
 {
  bone1Name=dyn_beltTop2_01_MSTR
  bone2Name=spine_01
  swing1LimitDegrees=45
  swing1Motion=Free
  swing2LimitDegrees=45
  swing2Motion=Free
  twistLimitDegrees=45
  twistMotion=Free
 }
 {bone1Name=dyn_beltTop2_01, bone2Name=dyn_beltTop2_01_MSTR, swing1LimitDegrees=5, swing2LimitDegrees=15, twistLimitDegrees=5}
 {bone1Name=dyn_beltGrenade2_01_l, bone2Name=dyn_beltTop2_01, swing1LimitDegrees=65, swing2LimitDegrees=65, twistLimitDegrees=5}
 {bone1Name=dyn_beltGrenade_01_l, bone2Name=dyn_beltTop2_01, swing1LimitDegrees=65, swing2LimitDegrees=65, twistLimitDegrees=5}
 {
  bone1Name=spine_02
  bone2Name=spine_01
  swing1LimitDegrees=45
  swing1Motion=Free
  swing2LimitDegrees=45
  swing2Motion=Free
  twistLimitDegrees=45
  twistMotion=Free
 }
 {
  bone1Name=spine_03
  bone2Name=spine_02
  swing1LimitDegrees=45
  swing1Motion=Free
  swing2LimitDegrees=45
  swing2Motion=Free
  twistLimitDegrees=45
  twistMotion=Free
 }
 {
  bone1Name=dyn_beltGrenade_01_l
  bone2Name=dyn_beltGrenade2_01_l
  swing1LimitDegrees=45
  swing1Motion=Locked
  swing2LimitDegrees=45
  swing2Motion=Locked
  twistLimitDegrees=45
  twistMotion=Locked
 }
]
