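# Physics asset for the F_MED Bullet Blue female character body, setting up secondary motion for the jacket and collar.
# Box bodies form one- or two-link chains with tight twist and swing1 limits and wider swing2; upper and mid jacket chains are right-side only.
# Each chain hangs from a tiny _MSTR anchor body, angularly free on spine_04 (most jacket chains) or spine_05 (collar, upper jacket).
# Spine_02 to spine_05 capsules form the torso.
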PhysicsAsset /BRCosmetics/Characters/Player/Female/Medium/Bodies/F_MED_BulletBlue/Meshes/F_MED_Bullet_Blue_Physics
bodies {
 dyn_collar_back_01 [{shapeType=Box, extentX=4.2985, extentY=2.4953, extentZ=6.1023, center=<2.0799 -0.0001 0.0001>, rotation=<rot 0 0 -90.0001>}]
 dyn_collar_back_01_MSTR [{shapeType=Box, extentX=0.01, extentY=0.01, extentZ=0.01}]
 dyn_collar_front_01_l [{shapeType=Box, extentX=4.2985, extentY=2.4953, extentZ=6.1023, center=<2.0799 -0.0001 0.0001>, rotation=<rot 0 0 -90.0001>}]
 dyn_collar_front_01_l_MSTR [{shapeType=Box, extentX=0.01, extentY=0.01, extentZ=0.01}]
 dyn_collar_front_01_r [{shapeType=Box, extentX=4.2985, extentY=2.4953, extentZ=6.1023, center=<-2.0799 0.0001 -0.0001>, rotation=<rot 0 0 89.9999>}]
 dyn_collar_front_01_r_MSTR [{shapeType=Box, extentX=0.01, extentY=0.01, extentZ=0.01}]
 dyn_collar_front_02_l [{shapeType=Box, extentX=4.2985, extentY=2.4953, extentZ=6.1023, center=<2.0799 -0.0001 0.0001>, rotation=<rot 0 0 -90.0001>}]
 dyn_collar_front_02_r [{shapeType=Box, extentX=4.2985, extentY=2.4953, extentZ=6.1023, center=<-2.0799 0.0001 -0.0001>, rotation=<rot 0 0 89.9999>}]
 dyn_collar_side_01_l [{shapeType=Box, extentX=4.2985, extentY=2.4953, extentZ=6.1023, center=<2.0799 -0.0001 0.0001>, rotation=<rot 0 0 -90.0001>}]
 dyn_collar_side_01_l_MSTR [{shapeType=Box, extentX=0.01, extentY=0.01, extentZ=0.01}]
 dyn_collar_side_01_r [{shapeType=Box, extentX=4.2985, extentY=2.4953, extentZ=6.1023, center=<-2.0799 0.0001 -0.0001>, rotation=<rot 0 0 89.9999>}]
 dyn_collar_side_01_r_MSTR [{shapeType=Box, extentX=0.01, extentY=0.01, extentZ=0.01}]
 dyn_jacket_back_01 [{shapeType=Box, extentX=5.7985, extentY=2.4953, extentZ=8.6023, center=<3.3387 0 0.0001>, rotation=<rot 0 0 -90.0001>}]
 dyn_jacket_back_01_MSTR [{shapeType=Box, extentX=0.01, extentY=0.01, extentZ=0.01}]
 dyn_jacket_back_02 [{shapeType=Box, extentX=5.7985, extentY=2.4953, extentZ=8.6023, center=<3.3387 0 0.0001>, rotation=<rot 0 0 -90.0001>}]
 dyn_jacket_front_01_l [{shapeType=Box, extentX=5.7985, extentY=2.4953, extentZ=8.6023, center=<3.3387 0 0.0001>, rotation=<rot 0 0 -90.0001>}]
 dyn_jacket_front_01_l_MSTR [{shapeType=Box, extentX=0.01, extentY=0.01, extentZ=0.01}]
 dyn_jacket_front_01_r [{shapeType=Box, extentX=5.7985, extentY=2.4953, extentZ=8.6023, center=<-3.3387 0 -0.0001>, rotation=<rot 0 0 89.9999>}]
 dyn_jacket_front_01_r_MSTR [{shapeType=Box, extentX=0.01, extentY=0.01, extentZ=0.01}]
 dyn_jacket_front_02_l [{shapeType=Box, extentX=5.7985, extentY=2.4953, extentZ=8.6023, center=<3.3387 0 0.0001>, rotation=<rot 0 0 -90.0001>}]
 dyn_jacket_front_02_r [{shapeType=Box, extentX=5.7985, extentY=2.4953, extentZ=8.6023, center=<-3.3387 0 -0.0001>, rotation=<rot 0 0 89.9999>}]
 dyn_jacket_mid_01_r [{shapeType=Box, extentX=4.2985, extentY=2.4953, extentZ=6.1023, center=<-2.0799 0.0001 -0.0001>, rotation=<rot 0 0 89.9999>}]
 dyn_jacket_mid_01_r_MSTR [{shapeType=Box, extentX=0.01, extentY=0.01, extentZ=0.01}]
 dyn_jacket_side_01_l [{shapeType=Box, extentX=5.7985, extentY=2.4953, extentZ=8.6023, center=<3.3387 0 0.0001>, rotation=<rot 0 0 -90.0001>}]
 dyn_jacket_side_01_l_MSTR [{shapeType=Box, extentX=0.01, extentY=0.01, extentZ=0.01}]
 dyn_jacket_side_01_r [{shapeType=Box, extentX=5.7985, extentY=2.4953, extentZ=8.6023, center=<-3.3387 0 -0.0001>, rotation=<rot 0 0 89.9999>}]
 dyn_jacket_side_01_r_MSTR [{shapeType=Box, extentX=0.01, extentY=0.01, extentZ=0.01}]
 dyn_jacket_side_02_l [{shapeType=Box, extentX=5.7985, extentY=2.4953, extentZ=8.6023, center=<3.3387 0 0.0001>, rotation=<rot 0 0 -90.0001>}]
 dyn_jacket_side_02_r [{shapeType=Box, extentX=5.7985, extentY=2.4953, extentZ=8.6023, center=<-3.3387 0 -0.0001>, rotation=<rot 0 0 89.9999>}]
 dyn_jacket_upper_01_r [{shapeType=Box, extentX=4.2985, extentY=2.4953, extentZ=6.1023, center=<-2.0799 0.0001 -0.0001>, rotation=<rot 0 0 89.9999>}]
 dyn_jacket_upper_01_r_MSTR [{shapeType=Box, extentX=0.01, extentY=0.01, extentZ=0.01}]
 dyn_jacket_upper_02_r [{shapeType=Box, extentX=4.2985, extentY=2.4953, extentZ=6.1023, center=<-2.0799 0.0001 -0.0001>, rotation=<rot 0 0 89.9999>}]
 spine_02 [{shapeType=Capsule, length=3.25, radius=8.1081, center=<5.181 3.2921 0>}]
 spine_03 [{shapeType=Capsule, length=0.5, radius=0.505}]
 spine_04 [{shapeType=Capsule, length=4.1866, radius=8.3143, center=<0 2.9968 0>}]
 spine_05 [{shapeType=Capsule, length=9.4601, radius=9.6763, center=<0 2.0392 0>}]
}
constraints [
 {bone1Name=dyn_jacket_upper_01_r, bone2Name=dyn_jacket_upper_01_r_MSTR, swing1LimitDegrees=5, swing2LimitDegrees=45, twistLimitDegrees=5}
 {bone1Name=dyn_jacket_upper_02_r, bone2Name=dyn_jacket_upper_01_r, swing1LimitDegrees=5, swing2LimitDegrees=45, twistLimitDegrees=5}
 {bone1Name=dyn_jacket_side_01_l, bone2Name=dyn_jacket_side_01_l_MSTR, swing1LimitDegrees=5, swing2LimitDegrees=20, twistLimitDegrees=5}
 {bone1Name=dyn_jacket_side_02_l, bone2Name=dyn_jacket_side_01_l, swing1LimitDegrees=5, swing2LimitDegrees=20, twistLimitDegrees=5}
 {bone1Name=dyn_jacket_front_01_l, bone2Name=dyn_jacket_front_01_l_MSTR, swing1LimitDegrees=5, swing2LimitDegrees=35, twistLimitDegrees=5}
 {bone1Name=dyn_jacket_front_02_l, bone2Name=dyn_jacket_front_01_l, swing1LimitDegrees=5, swing2LimitDegrees=35, twistLimitDegrees=5}
 {bone1Name=dyn_jacket_back_02, bone2Name=dyn_jacket_back_01, swing1LimitDegrees=5, swing2LimitDegrees=20, twistLimitDegrees=5}
 {bone1Name=dyn_jacket_back_01, bone2Name=dyn_jacket_back_01_MSTR, swing1LimitDegrees=5, swing2LimitDegrees=20, twistLimitDegrees=5}
 {bone1Name=dyn_jacket_mid_01_r, bone2Name=dyn_jacket_mid_01_r_MSTR, swing1LimitDegrees=5, swing2LimitDegrees=45, twistLimitDegrees=5}
 {bone1Name=dyn_collar_front_01_l, bone2Name=dyn_collar_front_01_l_MSTR, swing1LimitDegrees=5, swing2LimitDegrees=45, twistLimitDegrees=5}
 {bone1Name=dyn_collar_front_02_l, bone2Name=dyn_collar_front_01_l, swing1LimitDegrees=5, swing2LimitDegrees=45, twistLimitDegrees=5}
 {bone1Name=dyn_collar_side_01_l, bone2Name=dyn_collar_side_01_l_MSTR, swing1LimitDegrees=5, swing2LimitDegrees=45, twistLimitDegrees=5}
 {bone1Name=dyn_collar_back_01, bone2Name=dyn_collar_back_01_MSTR, swing1LimitDegrees=5, swing2LimitDegrees=45, twistLimitDegrees=5}
 {bone1Name=dyn_jacket_front_02_r, bone2Name=dyn_jacket_front_01_r, swing1LimitDegrees=5, swing2LimitDegrees=35, twistLimitDegrees=5}
 {bone1Name=dyn_jacket_front_01_r, bone2Name=dyn_jacket_front_01_r_MSTR, swing1LimitDegrees=5, swing2LimitDegrees=35, twistLimitDegrees=5}
 {bone1Name=dyn_jacket_side_01_r, bone2Name=dyn_jacket_side_01_r_MSTR, swing1LimitDegrees=5, swing2LimitDegrees=20, twistLimitDegrees=5}
 {bone1Name=dyn_jacket_side_02_r, bone2Name=dyn_jacket_side_01_r, swing1LimitDegrees=5, swing2LimitDegrees=20, twistLimitDegrees=5}
 {bone1Name=dyn_collar_front_01_r, bone2Name=dyn_collar_front_01_r_MSTR, swing1LimitDegrees=5, swing2LimitDegrees=45, twistLimitDegrees=5}
 {bone1Name=dyn_collar_front_02_r, bone2Name=dyn_collar_front_01_r, swing1LimitDegrees=5, swing2LimitDegrees=45, twistLimitDegrees=5}
 {bone1Name=dyn_collar_side_01_r, bone2Name=dyn_collar_side_01_r_MSTR, swing1LimitDegrees=5, swing2LimitDegrees=45, twistLimitDegrees=5}
 {
  bone1Name=dyn_collar_front_01_l_MSTR
  bone2Name=spine_05
  swing1LimitDegrees=45
  swing1Motion=Free
  swing2LimitDegrees=45
  swing2Motion=Free
  twistLimitDegrees=45
  twistMotion=Free
 }
 {
  bone1Name=dyn_collar_side_01_l_MSTR
  bone2Name=spine_05
  swing1LimitDegrees=45
  swing1Motion=Free
  swing2LimitDegrees=45
  swing2Motion=Free
  twistLimitDegrees=45
  twistMotion=Free
 }
 {
  bone1Name=dyn_collar_back_01_MSTR
  bone2Name=spine_05
  swing1LimitDegrees=45
  swing1Motion=Free
  swing2LimitDegrees=45
  swing2Motion=Free
  twistLimitDegrees=45
  twistMotion=Free
 }
 {
  bone1Name=dyn_collar_front_01_r_MSTR
  bone2Name=spine_05
  swing1LimitDegrees=45
  swing1Motion=Free
  swing2LimitDegrees=45
  swing2Motion=Free
  twistLimitDegrees=45
  twistMotion=Free
 }
 {
  bone1Name=dyn_collar_side_01_r_MSTR
  bone2Name=spine_05
  swing1LimitDegrees=45
  swing1Motion=Free
  swing2LimitDegrees=45
  swing2Motion=Free
  twistLimitDegrees=45
  twistMotion=Free
 }
 {
  bone1Name=dyn_jacket_upper_01_r_MSTR
  bone2Name=spine_05
  swing1LimitDegrees=45
  swing1Motion=Free
  swing2LimitDegrees=45
  swing2Motion=Free
  twistLimitDegrees=45
  twistMotion=Free
 }
 {
  bone1Name=dyn_jacket_side_01_l_MSTR
  bone2Name=spine_04
  swing1LimitDegrees=45
  swing1Motion=Free
  swing2LimitDegrees=45
  swing2Motion=Free
  twistLimitDegrees=45
  twistMotion=Free
 }
 {
  bone1Name=dyn_jacket_front_01_l_MSTR
  bone2Name=spine_04
  swing1LimitDegrees=45
  swing1Motion=Free
  swing2LimitDegrees=45
  swing2Motion=Free
  twistLimitDegrees=45
  twistMotion=Free
 }
 {
  bone1Name=dyn_jacket_back_01_MSTR
  bone2Name=spine_04
  swing1LimitDegrees=45
  swing1Motion=Free
  swing2LimitDegrees=45
  swing2Motion=Free
  twistLimitDegrees=45
  twistMotion=Free
 }
 {
  bone1Name=dyn_jacket_front_01_r_MSTR
  bone2Name=spine_04
  swing1LimitDegrees=45
  swing1Motion=Free
  swing2LimitDegrees=45
  swing2Motion=Free
  twistLimitDegrees=45
  twistMotion=Free
 }
 {
  bone1Name=dyn_jacket_side_01_r_MSTR
  bone2Name=spine_04
  swing1LimitDegrees=45
  swing1Motion=Free
  swing2LimitDegrees=45
  swing2Motion=Free
  twistLimitDegrees=45
  twistMotion=Free
 }
 {
  bone1Name=dyn_jacket_mid_01_r_MSTR
  bone2Name=spine_04
  swing1LimitDegrees=45
  swing1Motion=Free
  swing2LimitDegrees=45
  swing2Motion=Free
  twistLimitDegrees=45
  twistMotion=Free
 }
 {
  bone1Name=spine_03
  bone2Name=spine_02
  swing1LimitDegrees=45
  swing1Motion=Free
  swing2LimitDegrees=45
  swing2Motion=Free
  twistLimitDegrees=45
  twistMotion=Free
 }
 {
  bone1Name=spine_04
  bone2Name=spine_03
  swing1LimitDegrees=45
  swing1Motion=Free
  swing2LimitDegrees=45
  swing2Motion=Free
  twistLimitDegrees=45
  twistMotion=Free
 }
 {
  bone1Name=spine_05
  bone2Name=spine_04
  swing1LimitDegrees=45
  swing1Motion=Free
  swing2LimitDegrees=45
  swing2Motion=Free
  twistLimitDegrees=45
  twistMotion=Free
 }
]
