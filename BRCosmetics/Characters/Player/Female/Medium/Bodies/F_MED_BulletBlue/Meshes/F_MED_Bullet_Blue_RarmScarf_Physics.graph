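# Physics asset for the right-arm scarf on the Bullet Blue female medium character body cosmetic.
# A tiny box body on the master bone anchors a capsule on the scarf's base bone, which forks into two tails (A and B) of two short capsules each.
# Constraints link each bone to its parent with tight swing and twist limits, so the dynamic bones can bend only slightly at each joint.
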PhysicsAsset /BRCosmetics/Characters/Player/Female/Medium/Bodies/F_MED_BulletBlue/Meshes/F_MED_Bullet_Blue_RarmScarf_Physics
bodies {
 dyn_scarfA_01_r [{shapeType=Capsule, length=1.7603, radius=0.5, center=<0.9533 -0.0248 -0.0602>, rotation=<rot 0 -90.4149 -86.4267>}]
 dyn_scarfA_02_r [{shapeType=Capsule, length=1.7603, radius=0.5, center=<0.9533 -0.0248 -0.0602>, rotation=<rot 0 -90.4149 -86.4267>}]
 dyn_scarfB_01_r [{shapeType=Capsule, length=1.7603, radius=0.5, center=<0.9533 -0.0248 -0.0602>, rotation=<rot 0 -90.4149 -86.4267>}]
 dyn_scarfB_02_r [{shapeType=Capsule, length=1.7603, radius=0.5, center=<0.9533 -0.0248 -0.0602>, rotation=<rot 0 -90.4149 -86.4267>}]
 dyn_scarf_01_r [{shapeType=Capsule, length=3.7603, radius=0.5, center=<2.3084 -0.0346 -0.1449>, rotation=<rot 0 -90.4149 -86.4267>}]
 dyn_scarf_01_r_MSTR [{shapeType=Box, extentX=0.01, extentY=0.01, extentZ=0.01}]
}
constraints [
 {bone1Name=dyn_scarf_01_r, bone2Name=dyn_scarf_01_r_MSTR, swing1LimitDegrees=15, swing2LimitDegrees=15, twistLimitDegrees=5}
 {bone1Name=dyn_scarfA_01_r, bone2Name=dyn_scarf_01_r, swing1LimitDegrees=15, swing2LimitDegrees=15, twistLimitDegrees=5}
 {bone1Name=dyn_scarfA_02_r, bone2Name=dyn_scarfA_01_r, swing1LimitDegrees=15, swing2LimitDegrees=15, twistLimitDegrees=5}
 {bone1Name=dyn_scarfB_01_r, bone2Name=dyn_scarf_01_r, swing1LimitDegrees=15, swing2LimitDegrees=15, twistLimitDegrees=5}
 {bone1Name=dyn_scarfB_02_r, bone2Name=dyn_scarfB_01_r, swing1LimitDegrees=15, swing2LimitDegrees=15, twistLimitDegrees=5}
]
