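# Physics asset for the right foot of the female medium Bullet Blue cosmetic body, covering its dangling shoe chain.
# Three box bodies (chain 01, 03, 04) link in series with narrow swing and twist limits, hanging from a tiny master-bone box.
# That master and a separate shoe-dangle master, which has no child bodies, attach to the foot with free rotation.
# Two capsules form the foot body.
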PhysicsAsset /BRCosmetics/Characters/Player/Female/Medium/Bodies/F_MED_BulletBlue/Meshes/F_MED_Bullet_Blue_Rfoot_Physics
bodies {
 dyn_shoeChain_01_r [{shapeType=Box, extentX=4.7985, extentY=1.4953, extentZ=5.6023, center=<-2.3618 0 -0.0001>, rotation=<rot 0 0 180>}]
 dyn_shoeChain_01_r_MSTR [{shapeType=Box, extentX=0.01, extentY=0.01, extentZ=0.01}]
 dyn_shoeChain_03_r [{shapeType=Box, extentX=2.7985, extentY=0.9953, extentZ=3.1023, center=<-1.7018 0 -0.0001>, rotation=<rot 0 0 180>}]
 dyn_shoeChain_04_r [{shapeType=Box, extentX=2.7985, extentY=0.9953, extentZ=3.1023, center=<-1.7018 0 -0.0001>, rotation=<rot 0 0 180>}]
 dyn_shoeDangle_01_r_MSTR [{shapeType=Box, extentX=0.01, extentY=0.01, extentZ=0.01}]
 foot_r [
  {shapeType=Capsule, length=9.4056, radius=4.9439, center=<4.9674 -4.5155 0.0003>, rotation=<rot 0.0001 45.0003 -90.0001>}
  {shapeType=Capsule, length=7.75, radius=4.75, center=<5.8661 -9.8251 0>, rotation=<rot 0 0 -90>}
 ]
}
constraints [
 {bone1Name=dyn_shoeChain_01_r, bone2Name=dyn_shoeChain_01_r_MSTR, swing1LimitDegrees=25, swing2LimitDegrees=25, twistLimitDegrees=5}
 {bone1Name=dyn_shoeChain_03_r, bone2Name=dyn_shoeChain_01_r, swing1LimitDegrees=25, swing2LimitDegrees=5, twistLimitDegrees=5}
 {bone1Name=dyn_shoeChain_04_r, bone2Name=dyn_shoeChain_03_r, swing1LimitDegrees=25, swing2LimitDegrees=5, twistLimitDegrees=5}
 {
  bone1Name=dyn_shoeChain_01_r_MSTR
  bone2Name=foot_r
  swing1LimitDegrees=45
  swing1Motion=Free
  swing2LimitDegrees=45
  swing2Motion=Free
  twistLimitDegrees=45
  twistMotion=Free
 }
 {
  bone1Name=dyn_shoeDangle_01_r_MSTR
  bone2Name=foot_r
  swing1LimitDegrees=45
  swing1Motion=Free
  swing2LimitDegrees=45
  swing2Motion=Free
  twistLimitDegrees=45
  twistMotion=Free
 }
]
