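# Physics asset for the right-side gun belt on the female medium Bullet Blue player cosmetic, covering belt straps, gun and three bullets.
# Capsules model pelvis, lower spine and right thigh, while tiny "MSTR" bodies link belt pieces to pelvis and spine.
# The gun hangs from the mid belt and the bullets from the bullet belt with angular limits.
# Free-angle constraints cross-link neighboring straps, forming loops rather than simple chains.
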PhysicsAsset /BRCosmetics/Characters/Player/Female/Medium/Bodies/F_MED_BulletBlue/Meshes/F_MED_Bullet_Blue_RgunBelt_Physics
bodies {
 dyn_beltBullter_01_r [{shapeType=Box, extentX=13.5, extentY=1.5, extentZ=7, center=<8.0376 -0.0001 0.0001>}]
 dyn_beltBullter_01_r_MSTR [{shapeType=Capsule, length=0.01, radius=0.01}]
 dyn_beltTop5_01_MSTR [{shapeType=Box, extentX=0.01, extentY=0.01, extentZ=0.01}]
 dyn_beltTop8_01_MSTR [{shapeType=Box, extentX=0.01, extentY=0.01, extentZ=0.01}]
 dyn_beltTop8_02 [{shapeType=Box, extentX=2.3824, extentY=4.1031, extentZ=8.5281, center=<-0.0417 0.7114 -0.1704>, rotation=<rot 0 -164.193 -76.9354>}]
 dyn_bullet1_01_r [{shapeType=Capsule, length=1.9295, radius=3.3889, center=<2.5133 -2.0913 0.0534>, rotation=<rot 0 -87.047 -87.9004>}]
 dyn_bullet2_01_r [{shapeType=Capsule, length=1.9295, radius=3.3889, center=<2.5133 -2.0913 0.0534>, rotation=<rot 0 -87.047 -87.9004>}]
 dyn_bullet3_01_r [{shapeType=Capsule, length=1.9295, radius=3.3889, center=<2.5133 -2.0913 0.0534>, rotation=<rot 0 -87.047 -87.9004>}]
 dyn_gun_01_r [{shapeType=Capsule, length=22.1101, radius=3.3594, center=<12.4144 0.7591 2.9254>, rotation=<rot -0.0475 88.3524 -89.8225>}]
 dyn_gunbeltBack_01_r [{shapeType=Box, extentX=13.51, extentY=5.51, extentZ=2.5, center=<6.1182 0.0002 0>, rotation=<rot 0 0 -90>}]
 dyn_gunbeltBack_01_r_MSTR [{shapeType=Box, extentX=0.01, extentY=0.01, extentZ=0.01}]
 dyn_gunbeltFront_01_r [{shapeType=Box, extentX=11.01, extentY=5.51, extentZ=2, center=<5.9228 0.0001 0>, rotation=<rot 0 0 -90>}]
 dyn_gunbeltMid_01_r [{shapeType=Box, extentX=15, extentY=8.5, extentZ=0.9, center=<7.8242 0 0.0001>}]
 dyn_gunbeltMid_01_r_MSTR [{shapeType=Box, extentX=0.01, extentY=0.01, extentZ=0.01}]
 pelvis [{shapeType=Capsule, length=14.5474, radius=8.677, center=<-2.7211 -1.7752 0.0202>, rotation=<rot 0.7619 -90 -0.0686>}]
 spine_01 [{shapeType=Capsule, length=8.9401, radius=7.1131, center=<3.0181 1.3101 0.0187>, rotation=<rot -1.1012 -90 0.5271>}]
 spine_02 [{shapeType=Capsule, length=6.4401, radius=7.1131, center=<3.669 3.109 0.0473>, rotation=<rot -1.1012 -90 0.5271>}]
 thigh_r [{shapeType=Capsule, length=25.1062, radius=7.6185, center=<12.029 -0.0002 -0.9027>, rotation=<rot 90 0 -179.9999>}]
 thigh_twist_01_r [{shapeType=Capsule, length=25.1062, radius=7.6185, center=<-15.2116 2.2947 -0.0004>, rotation=<rot 90 -0.0002 180>}]
}
constraints [
 {bone1Name=dyn_gunbeltFront_01_r, bone2Name=dyn_beltTop8_02, swing1LimitDegrees=45, swing2LimitDegrees=45, twistLimitDegrees=5}
 {bone1Name=dyn_gunbeltBack_01_r, bone2Name=dyn_gunbeltBack_01_r_MSTR, swing1LimitDegrees=45, swing2LimitDegrees=15, twistLimitDegrees=5}
 {bone1Name=dyn_gunbeltMid_01_r, bone2Name=dyn_gunbeltMid_01_r_MSTR, swing1LimitDegrees=45, swing2LimitDegrees=45, twistLimitDegrees=5}
 {
  bone1Name=dyn_gunbeltBack_01_r_MSTR
  bone2Name=dyn_beltTop5_01_MSTR
  swing1LimitDegrees=45
  swing1Motion=Free
  swing2LimitDegrees=45
  swing2Motion=Free
  twistLimitDegrees=45
  twistMotion=Free
 }
 {bone1Name=dyn_beltBullter_01_r, bone2Name=dyn_beltBullter_01_r_MSTR, swing1LimitDegrees=60, swing2LimitDegrees=60, twistLimitDegrees=5}
 {
  bone1Name=dyn_beltTop5_01_MSTR
  bone2Name=spine_01
  swing1LimitDegrees=45
  swing1Motion=Free
  swing2LimitDegrees=45
  swing2Motion=Free
  twistLimitDegrees=45
  twistMotion=Free
 }
 {
  bone1Name=dyn_beltTop8_01_MSTR
  bone2Name=spine_01
  swing1LimitDegrees=45
  swing1Motion=Free
  swing2LimitDegrees=45
  swing2Motion=Free
  twistLimitDegrees=45
  twistMotion=Free
 }
 {
  bone1Name=thigh_r
  bone2Name=pelvis
  swing1LimitDegrees=45
  swing1Motion=Free
  swing2LimitDegrees=45
  swing2Motion=Free
  twistLimitDegrees=45
  twistMotion=Free
 }
 {
  bone1Name=dyn_gunbeltMid_01_r_MSTR
  bone2Name=pelvis
  swing1LimitDegrees=45
  swing1Motion=Free
  swing2LimitDegrees=45
  swing2Motion=Free
  twistLimitDegrees=45
  twistMotion=Free
 }
 {
  bone1Name=dyn_beltBullter_01_r_MSTR
  bone2Name=pelvis
  swing1LimitDegrees=45
  swing1Motion=Free
  swing2LimitDegrees=45
  swing2Motion=Free
  twistLimitDegrees=45
  twistMotion=Free
 }
 {
  bone1Name=thigh_twist_01_r
  bone2Name=thigh_r
  swing1LimitDegrees=45
  swing1Motion=Free
  swing2LimitDegrees=45
  swing2Motion=Free
  twistLimitDegrees=45
  twistMotion=Free
 }
 {
  bone1Name=dyn_gunbeltBack_01_r
  bone2Name=dyn_gunbeltMid_01_r
  swing1LimitDegrees=45
  swing1Motion=Free
  swing2LimitDegrees=45
  swing2Motion=Free
  twistLimitDegrees=45
  twistMotion=Free
 }
 {
  bone1Name=dyn_beltBullter_01_r
  bone2Name=dyn_gunbeltMid_01_r
  swing1LimitDegrees=45
  swing1Motion=Free
  swing2LimitDegrees=45
  swing2Motion=Free
  twistLimitDegrees=45
  twistMotion=Free
 }
 {bone1Name=dyn_bullet2_01_r, bone2Name=dyn_beltBullter_01_r, swing1LimitDegrees=75, swing2LimitDegrees=15, twistLimitDegrees=5}
 {bone1Name=dyn_bullet1_01_r, bone2Name=dyn_beltBullter_01_r, swing1LimitDegrees=75, swing2LimitDegrees=15, twistLimitDegrees=5}
 {bone1Name=dyn_bullet3_01_r, bone2Name=dyn_beltBullter_01_r, swing1LimitDegrees=75, swing2LimitDegrees=15, twistLimitDegrees=5}
 {
  bone1Name=dyn_beltBullter_01_r
  bone2Name=dyn_gunbeltFront_01_r
  swing1LimitDegrees=45
  swing1Motion=Free
  swing2LimitDegrees=45
  swing2Motion=Free
  twistLimitDegrees=45
  twistMotion=Free
 }
 {
  bone1Name=dyn_beltTop8_02
  bone2Name=dyn_beltTop8_01_MSTR
  swing1LimitDegrees=45
  swing1Motion=Free
  swing2LimitDegrees=45
  swing2Motion=Free
  twistLimitDegrees=45
  twistMotion=Free
 }
 {
  bone1Name=dyn_beltTop8_02
  bone2Name=dyn_gunbeltFront_01_r
  swing1LimitDegrees=45
  swing1Motion=Free
  swing2LimitDegrees=45
  swing2Motion=Free
  twistLimitDegrees=45
  twistMotion=Free
 }
 {bone1Name=dyn_gun_01_r, bone2Name=dyn_gunbeltMid_01_r, swing1LimitDegrees=45, swing2LimitDegrees=8, twistLimitDegrees=8}
 {
  bone1Name=spine_01
  bone2Name=pelvis
  swing1LimitDegrees=45
  swing1Motion=Free
  swing2LimitDegrees=45
  swing2Motion=Free
  twistLimitDegrees=45
  twistMotion=Free
 }
 {
  bone1Name=spine_02
  bone2Name=spine_01
  swing1LimitDegrees=45
  swing1Motion=Free
  swing2LimitDegrees=45
  swing2Motion=Free
  twistLimitDegrees=45
  twistMotion=Free
 }
]
