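# Post-style AnimBP for the F_MED Bullet Blue character body, copying the attached parent's pose and animating its belts, jacket, collar and scarf.
# Constraints and additive bone offsets driven by arm swing, crouch, backpack and skydive state combine with RigidBody sims, pushed by skydive wind, into the SimCache pose.
# When rigid bodies are disabled or LOD exceeds 1, output adds fixed belt corrections.
# Arm-cuff and foot sims are unconnected to the output.
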AnimBlueprint /BRCosmetics/Characters/Player/Female/Medium/Bodies/F_MED_BulletBlue/Meshes/F_MED_Bullet_Blue_AnimBP
target_skeleton=&/F_MED_Bullet_Blue_Skeleton
graphs [
 AnimGraph {nodes [
   n1 = Get windVector() -> (windVector:Vector)
   RigidBody(ExternalForce=@n1) {
    bClampLinearTranslationLimitToRefPose=true
    bOverrideWorldGravity=true
    overridePhysicsAsset=&/F_MED_Bullet_Blue_LarmCuff_Physics
    simulationSpace=BaseBoneSpace
    alphaBoolBlend {blendOption=Linear}
    baseBoneRef {boneName=upperarm_l}
    componentAppliedLinearAccClamp=<3000 3000 3000>
    componentLinearAccScale=<0.5 0.5 0.5>
    componentLinearVelScale=<0.7 0.7 0.7>
    overrideWorldGravity {z=-980}
   }
   n2 = RigidBody(ExternalForce=@n1) {
    bClampLinearTranslationLimitToRefPose=true
    bOverrideWorldGravity=true
    overridePhysicsAsset=&/F_MED_Bullet_Blue_Lfoot_Physics
    simulationSpace=BaseBoneSpace
    alphaBoolBlend {blendOption=Linear}
    baseBoneRef {boneName=foot_l}
    componentAppliedLinearAccClamp=<3000 3000 3000>
    componentLinearAccScale=<0.5 0.5 0.5>
    componentLinearVelScale=<0.7 0.7 0.7>
    overrideWorldGravity {z=-980}
   }
   RigidBody(ComponentPose=@n2, ExternalForce=@n1) {
    bClampLinearTranslationLimitToRefPose=true
    bOverrideWorldGravity=true
    overridePhysicsAsset=&/F_MED_Bullet_Blue_Rfoot_Physics
    simulationSpace=BaseBoneSpace
    alphaBoolBlend {blendOption=Linear}
    baseBoneRef {boneName=foot_r}
    componentAppliedLinearAccClamp=<3000 3000 3000>
    componentLinearAccScale=<0.5 0.5 0.5>
    componentLinearVelScale=<0.7 0.7 0.7>
    overrideWorldGravity {z=-980}
   }
   n3 = "Constraint - dyn_beltBottom3_02_l (PPP)"!(
    ComponentPose=Local To Component(LocalPose=Use cached pose 'SimCache'())
    Alpha=1
    ConstraintWeights_0=0.218
    ConstraintWeights_1:float=0.6213
    ConstraintWeights_2:float=0.1607
   ) -> (Pose) {
    alphaBoolBlend {blendOption=Linear}
    boneToModify {boneName=dyn_beltBottom3_02_l}
    constraintSetup [{targetBone {boneName=thigh_r}}, {targetBone {boneName=thigh_twist_01_r}}, {targetBone {boneName=pelvis}}]
    constraintWeights [1,1,1]
   }
   n4 = "Transform (Modify) Bone - Bone: dyn_beltBottom3_02_l"!(ComponentPose=@n3, Alpha=1, Translation="0,0,0", Rotation="10.000,-0.000,0.000") -> (Pose) {
    rotationMode=BMM_Additive
    rotationSpace=BCS_BoneSpace
    translationMode=BMM_Additive
    translationSpace=BCS_BoneSpace
    alphaBoolBlend {blendOption=Linear}
    boneToModify {boneName=dyn_beltBottom3_02_l}
    rotation=<rot 10 0 0>
   }
   n5 = "Transform (Modify) Bone - Bone: dyn_beltPouch_01_l"!(ComponentPose=@n4, Alpha=1, Translation="0,0,0", Rotation="20,-5.5,-5.5") -> (Pose) {
    rotationMode=BMM_Additive
    rotationSpace=BCS_BoneSpace
    translationMode=BMM_Additive
    translationSpace=BCS_BoneSpace
    alphaBoolBlend {blendOption=Linear}
    boneToModify {boneName=dyn_beltPouch_01_l}
    rotation=<rot 20 -5.5 -5.5>
   }
   n6 = "Constraint - dyn_beltPouch_01_l (PPPPPPP)"!(
    ComponentPose=@n5
    Alpha=1
    ConstraintWeights_0=0.0001
    ConstraintWeights_1:float=0.004
    ConstraintWeights_2:float=0.0092
    ConstraintWeights_3:float=0.2549
    ConstraintWeights_4:float=0
    ConstraintWeights_5:float=0
    ConstraintWeights_6:float=0.7318
   ) -> (Pose) {
    alphaBoolBlend {blendOption=Linear}
    boneToModify {boneName=dyn_beltPouch_01_l}
    constraintSetup [
     {targetBone {boneName=dyn_beltGrenade2_01_l}}
     {targetBone {boneName=dyn_beltTop8_02}}
     {targetBone {boneName=dyn_beltGrenade_01_l}}
     {targetBone {boneName=dyn_beltTop2_02}}
     {targetBone {boneName=dyn_beltTop7_02}}
     {targetBone {boneName=dyn_beltTop3_02}}
     {targetBone {boneName=dyn_beltTop1_02}}
    ]
    constraintWeights [1,1,1,1,1,1,1]
   }
   n7 = "Transform (Modify) Bone - Bone: dyn_beltPouch_01_l"(ComponentPose=@n6, Alpha=1, Translation="0,0,0", Rotation="15,-14,-25") {
    rotationMode=BMM_Additive
    rotationSpace=BCS_BoneSpace
    translationMode=BMM_Additive
    translationSpace=BCS_BoneSpace
    alphaBoolBlend {blendOption=Linear}
    alphaScaleBiasClamp {bInterpResult=true, interpSpeedDecreasing=30, interpSpeedIncreasing=15}
    boneToModify {boneName=dyn_beltPouch_01_l}
    rotation=<rot 15 -14 -25>
   }
   n8 = "Constraint - dyn_beltBullter_02_r (PPPPPP)"!(
    ComponentPose=@n7
    Alpha=1
    ConstraintWeights_0=0
    ConstraintWeights_1:float=0.0001
    ConstraintWeights_2:float=0.0206
    ConstraintWeights_3:float=0.3617
    ConstraintWeights_4:float=0.0012
    ConstraintWeights_5:float=0.6165
   ) -> (Pose) {
    alphaBoolBlend {blendOption=Linear}
    boneToModify {boneName=dyn_beltBullter_02_r}
    constraintSetup [
     {targetBone {boneName=spine_01}}
     {targetBone {boneName=spine_02}}
     {targetBone {boneName=thigh_r}}
     {targetBone {boneName=thigh_twist_01_r}}
     {targetBone {boneName=thigh_twist_01_l}}
     {targetBone {boneName=pelvis}}
    ]
    constraintWeights [1,1,1,1,1,1]
   }
   n9 = "Constraint - dyn_beltTop5_02 (PPPPP)"!(
    ComponentPose=@n8
    Alpha=1
    ConstraintWeights_0=0.2189
    ConstraintWeights_1:float=0.2806
    ConstraintWeights_2:float=0.0463
    ConstraintWeights_3:float=0.041
    ConstraintWeights_4:float=0.4133
   ) -> (Pose) {
    alphaBoolBlend {blendOption=Linear}
    boneToModify {boneName=dyn_beltTop5_02}
    constraintSetup [
     {targetBone {boneName=spine_01}}
     {targetBone {boneName=spine_02}}
     {targetBone {boneName=spine_03}}
     {targetBone {boneName=spine_04}}
     {targetBone {boneName=pelvis}}
    ]
    constraintWeights [1,1,1,1,1]
   }
   n10 = "Constraint - dyn_beltTop6_02 (PPPPPP)"!(
    ComponentPose=@n9
    Alpha=1
    ConstraintWeights_0=0.1518
    ConstraintWeights_1:float=0.2639
    ConstraintWeights_2:float=0.0405
    ConstraintWeights_3:float=0.1011
    ConstraintWeights_4:float=0.0001
    ConstraintWeights_5:float=0.4426
   ) -> (Pose) {
    alphaBoolBlend {blendOption=Linear}
    boneToModify {boneName=dyn_beltTop6_02}
    constraintSetup [
     {targetBone {boneName=spine_01}}
     {targetBone {boneName=spine_02}}
     {targetBone {boneName=spine_03}}
     {targetBone {boneName=spine_04}}
     {targetBone {boneName=thigh_twist_01_r}}
     {targetBone {boneName=pelvis}}
    ]
    constraintWeights [1,1,1,1,1,1]
   }
   n11 = "Transform (Modify) Bone - Bone: dyn_beltPouch_01_l"(ComponentPose=@n10, Alpha=1, Translation="0,0,0", Rotation="39.816,-0.101,-24.927") {
    rotationMode=BMM_Additive
    rotationSpace=BCS_BoneSpace
    translationMode=BMM_Additive
    translationSpace=BCS_BoneSpace
    alphaBoolBlend {blendOption=Linear}
    alphaScaleBiasClamp {bInterpResult=true, interpSpeedDecreasing=30, interpSpeedIncreasing=15}
    boneToModify {boneName=dyn_beltPouch_01_l}
    rotation=<rot 39.816 -0.101 -24.927>
   }
   n12 = "Transform (Modify) Bone - Bone: dyn_beltBullter_02_r"!(ComponentPose=@n11, Alpha=1, Translation="-2,0,2", Rotation="-0.064,-100,0.053") -> (Pose) {
    rotationMode=BMM_Additive
    rotationSpace=BCS_BoneSpace
    translationMode=BMM_Additive
    translationSpace=BCS_BoneSpace
    alphaBoolBlend {blendOption=Linear}
    alphaScaleBiasClamp {bInterpResult=true, interpSpeedDecreasing=30, interpSpeedIncreasing=15}
    boneToModify {boneName=dyn_beltBullter_02_r}
    rotation=<rot -0.064 -100 0.053>
    translation {x=-2, z=2}
   }
   n13 = "Transform (Modify) Bone - Bone: dyn_beltPouch_01_l"(ComponentPose=@n12, Alpha=1, Translation="0,0,0", Rotation="37.494,-16.723,-54.582") {
    rotationMode=BMM_Additive
    rotationSpace=BCS_BoneSpace
    translationMode=BMM_Additive
    translationSpace=BCS_BoneSpace
    alphaBoolBlend {blendOption=Linear}
    alphaScaleBiasClamp {bInterpResult=true, interpSpeedDecreasing=30, interpSpeedIncreasing=15}
    boneToModify {boneName=dyn_beltPouch_01_l}
    rotation=<rot 37.494 -16.723 -54.582>
   }
   Output Pose(Result=Blend Poses by bool(
     BlendPose_0=Component To Local(ComponentPose=@n13)
     BlendPose_1=Use cached pose 'SimCache'()
     bActiveValue=Get rigidsOnOffBool() -> (rigidsOnOffBool:bool)
    ) {blendType=Linear}) {layerGroup=None}
   n14 = "Constraint - dyn_gunbeltMid_01_r_MSTR (RR)"!(
    ComponentPose=Local To Component(LocalPose=Copy Pose From Mesh() {bUseAttachedParent=true})
    Alpha=1
    ConstraintWeights_0=0
    ConstraintWeights_1:float=1
   ) -> (Pose) {
    alphaBoolBlend {blendOption=Linear}
    boneToModify {boneName=dyn_gunbeltMid_01_r_MSTR}
    constraintSetup [{transformType=Rotation, targetBone {boneName=thigh_twist_01_r}}, {transformType=Rotation, targetBone {boneName=pelvis}}]
    constraintWeights [1,1]
   }
   n15 = "Constraint - dyn_beltBullter_01_r_MSTR (PP)"!(ComponentPose=@n14, Alpha=1, ConstraintWeights_0=1, ConstraintWeights_1:float=0) -> (Pose) {
    alphaBoolBlend {blendOption=Linear}
    boneToModify {boneName=dyn_beltBullter_01_r_MSTR}
    constraintSetup [{targetBone {boneName=thigh_r}}, {targetBone {boneName=pelvis}}]
    constraintWeights [1,1]
   }
   n16 = "Transform (Modify) Bone - Bone: dyn_beltBullter_01_r_MSTR"(ComponentPose=@n15, Alpha=Get isCrouchingAlpha() -> (isCrouchingAlpha:double), Translation="4,0,0", Rotation="0,0,0", Scale="1,1,1") {
    translationMode=BMM_Additive
    translationSpace=BCS_BoneSpace
    alphaBoolBlend {blendOption=Linear}
    boneToModify {boneName=dyn_beltBullter_01_r_MSTR}
    translation {x=4}
   }
   n17 = "Constraint - dyn_gun_01_r (R)"(ComponentPose=@n16, Alpha=0, ConstraintWeights_0=1) {
    alphaBoolBlend {blendOption=Linear}
    boneToModify {boneName=dyn_gun_01_r}
    constraintSetup [{transformType=Rotation, targetBone {boneName=pelvis}}]
    constraintWeights [1]
   }
   n18 = Get armSwingDelta_r() -> (armSwingDelta_r:double)
   n19 = "Transform (Modify) Bone - Bone: dyn_collar_front_01_r_MSTR"(ComponentPose=@n17, Alpha=@n18, Translation="0,0,0", Rotation="15.000,0.000,30.000", Scale="1,1,1") {
    rotationMode=BMM_Additive
    rotationSpace=BCS_BoneSpace
    alphaBoolBlend {blendOption=Linear}
    boneToModify {boneName=dyn_collar_front_01_r_MSTR}
    rotation {pitch=15, roll=30}
   }
   n20 = "Transform (Modify) Bone - Bone: dyn_collar_side_01_r_MSTR"(ComponentPose=@n19, Alpha=@n18, Translation="0,0,0", Rotation="25.000,0.000,0.000", Scale="1,1,1") {
    rotationMode=BMM_Additive
    rotationSpace=BCS_BoneSpace
    alphaBoolBlend {blendOption=Linear}
    boneToModify {boneName=dyn_collar_side_01_r_MSTR}
    rotation {pitch=25}
   }
   n21 = "Transform (Modify) Bone - Bone: dyn_jacket_upper_01_r_MSTR"(ComponentPose=@n20, Alpha=@n18, Translation="0,0,0", Rotation="-15.000,0.000,0.000", Scale="1,1,1") {
    rotationMode=BMM_Additive
    rotationSpace=BCS_BoneSpace
    alphaBoolBlend {blendOption=Linear}
    boneToModify {boneName=dyn_jacket_upper_01_r_MSTR}
    rotation {pitch=-15}
   }
   n22 = Get armSwingDelta_l() -> (armSwingDelta_l:double)
   n23 = "Transform (Modify) Bone - Bone: dyn_collar_front_01_l_MSTR"(ComponentPose=@n21, Alpha=@n22, Translation="0,0,0", Rotation="15,0.000,5.000", Scale="1,1,1") {
    rotationMode=BMM_Additive
    rotationSpace=BCS_BoneSpace
    alphaBoolBlend {blendOption=Linear}
    boneToModify {boneName=dyn_collar_front_01_l_MSTR}
    rotation {pitch=15, roll=5}
   }
   n24 = "Transform (Modify) Bone - Bone: dyn_collar_side_01_l_MSTR"(ComponentPose=@n23, Alpha=@n22, Translation="0,0,0", Rotation="25.000,0.000,0.000", Scale="1,1,1") {
    rotationMode=BMM_Additive
    rotationSpace=BCS_BoneSpace
    alphaBoolBlend {blendOption=Linear}
    boneToModify {boneName=dyn_collar_side_01_l_MSTR}
    rotation {pitch=25}
   }
   n25 = Get isCrouchingAlpha()
   n26 = "Transform (Modify) Bone - Bone: dyn_beltTop2_01_MSTR"(ComponentPose=@n24, Alpha=@n25, Translation="-1.255,0.339,-0.214", Rotation="4.354,-14.524,-5.093", Scale="1,1,1") {
    rotationMode=BMM_Additive
    rotationSpace=BCS_BoneSpace
    translationMode=BMM_Additive
    translationSpace=BCS_BoneSpace
    alphaBoolBlend {blendOption=Linear}
    boneToModify {boneName=dyn_beltTop2_01_MSTR}
    rotation=<rot 4.354 -14.5242 -5.093>
    translation=<-1.2553 0.3391 -0.2141>
   }
   n27 = "Transform (Modify) Bone - Bone: dyn_beltTop3_01_MSTR"(ComponentPose=@n26, Alpha=@n25, Translation="-0.366,0.063,-0.044", Rotation="-5.000,-10.000,-0.000", Scale="1,1,1") {
    rotationMode=BMM_Additive
    rotationSpace=BCS_BoneSpace
    translationMode=BMM_Additive
    translationSpace=BCS_BoneSpace
    alphaBoolBlend {blendOption=Linear}
    boneToModify {boneName=dyn_beltTop3_01_MSTR}
    rotation=<rot -5 -10.0002 0>
    translation=<-0.366 0.063 -0.044>
   }
   n28 = "Transform (Modify) Bone - Bone: dyn_beltTop1_01_MSTR"(ComponentPose=@n27, Alpha=@n25, Translation="4.423,2.371,1.674", Rotation="21.680,-28.959,-32.790", Scale="1,1,1") {
    rotationMode=BMM_Additive
    rotationSpace=BCS_BoneSpace
    translationMode=BMM_Additive
    translationSpace=BCS_BoneSpace
    alphaBoolBlend {blendOption=Linear}
    boneToModify {boneName=dyn_beltTop1_01_MSTR}
    rotation=<rot 21.68 -28.959 -32.7901>
    translation=<4.423 2.371 1.674>
   }
   n29 = RigidBody(ComponentPose=@n28, ExternalForce=@n1) {
    bClampLinearTranslationLimitToRefPose=true
    bOverrideWorldGravity=true
    overridePhysicsAsset=&/F_MED_Bullet_Blue_Physics
    simulationSpace=BaseBoneSpace
    alphaBoolBlend {blendOption=Linear}
    baseBoneRef {boneName=spine_03}
    componentAppliedLinearAccClamp=<3000 3000 3000>
    componentLinearAccScale=<0.5 0.5 0.5>
    componentLinearVelScale=<0.7 0.7 0.7>
    overrideWorldGravity {z=-980}
   }
   n30 = RigidBody(ComponentPose=@n29, ExternalForce=@n1) {
    bClampLinearTranslationLimitToRefPose=true
    bOverrideWorldGravity=true
    overridePhysicsAsset=&/F_MED_Bullet_Blue_RarmScarf_Physics
    simulationSpace=BaseBoneSpace
    alphaBoolBlend {blendOption=Linear}
    baseBoneRef {boneName=upperarm_r}
    componentAppliedLinearAccClamp=<3000 3000 3000>
    componentLinearAccScale=<0.5 0.5 0.5>
    componentLinearVelScale=<0.7 0.7 0.7>
    overrideWorldGravity {z=-980}
   }
   n31 = RigidBody(ComponentPose=@n30, ExternalForce=@n1) {
    bClampLinearTranslationLimitToRefPose=true
    bOverrideWorldGravity=true
    overridePhysicsAsset=&/F_MED_Bullet_Blue_LgunBelt_Physics
    simulationSpace=BaseBoneSpace
    alphaBoolBlend {blendOption=Linear}
    baseBoneRef {boneName=pelvis}
    componentAppliedLinearAccClamp=<1500 1500 400>
    componentLinearAccScale=<0.5 0.5 0.3>
    componentLinearVelScale=<0.4 0.4 0.4>
    overrideWorldGravity {z=-980}
   }
   n32 = "Constraint - dyn_gunbeltFront_01_r_MSTR (P)"(ComponentPose=@n31, Alpha=1, ConstraintWeights_0=1) {
    alphaBoolBlend {blendOption=Linear}
    boneToModify {boneName=dyn_gunbeltFront_01_r_MSTR}
    constraintSetup [{targetBone {boneName=dyn_beltTop8_02}}]
    constraintWeights [1]
   }
   n33 = RigidBody(ComponentPose=@n32, ExternalForce=@n1) {
    bClampLinearTranslationLimitToRefPose=true
    bOverrideWorldGravity=true
    overridePhysicsAsset=&/F_MED_Bullet_Blue_RgunBelt_Physics
    simulationSpace=BaseBoneSpace
    alphaBoolBlend {blendOption=Linear}
    baseBoneRef {boneName=thigh_twist_01_r}
    componentAppliedLinearAccClamp=<3000 3000 1500>
    componentLinearAccScale=<0.5 0.5 0.3>
    componentLinearVelScale=<0.7 0.7 0.6>
    overrideWorldGravity {z=-980}
   }
   n34 = "Constraint - dyn_beltTop4_01_MSTR (R)"(ComponentPose=@n33, Alpha=1, ConstraintWeights_0=1) {
    alphaBoolBlend {blendOption=Linear}
    boneToModify {boneName=dyn_beltTop4_01_MSTR}
    constraintSetup [{transformType=Rotation, targetBone {boneName=dyn_beltTop3_01}}]
    constraintWeights [1]
   }
   n35 = "Constraint - dyn_beltBottom4_02_l (PP)"!(ComponentPose=@n34, Alpha=1, ConstraintWeights_0=0.5, ConstraintWeights_1:float=0.5) -> (Pose) {
    alphaBoolBlend {blendOption=Linear}
    boneToModify {boneName=dyn_beltBottom4_02_l}
    constraintSetup [{targetBone {boneName=dyn_gunbeltBack_02_r}}, {targetBone {boneName=dyn_gunbeltMid_02_r}}]
    constraintWeights [1,1]
   }
   n36 = "Transform (Modify) Bone - Bone: dyn_beltBottom4_02_l"(ComponentPose=@n35, Alpha=1, Translation="-2,0,0", Rotation="0,0,0", Scale="1,1,1") {
    translationMode=BMM_Additive
    translationSpace=BCS_BoneSpace
    alphaBoolBlend {blendOption=Linear}
    boneToModify {boneName=dyn_beltBottom4_02_l}
    translation {x=-2}
   }
   n37 = Look At - Bone: dyn_gunbeltFront_01_r_MSTR(ComponentPose=@n36, Alpha=1) {
    alphaBoolBlend {blendOption=Linear}
    boneToModify {boneName=dyn_gunbeltFront_01_r_MSTR}
    lookAtLocation {x=-1}
    lookAtTarget {boneReference {boneName=dyn_bullet1_01_r}}
    lookAt_Axis {axis=<1 0>}
    lookUp_Axis {axis {x=0, z=1}}
   }
   n38 = "Constraint - dyn_beltTop6_02 (P)"(ComponentPose=@n37, Alpha=1, ConstraintWeights_0=1) {
    alphaBoolBlend {blendOption=Linear}
    boneToModify {boneName=dyn_beltTop6_02}
    constraintSetup [{targetBone {boneName=spine_02}}]
    constraintWeights [1]
   }
   n39 = "Transform (Modify) Bone - Bone: upperarm_twist_01_r"(ComponentPose=@n38, Alpha=Get shoulderInflate_r() -> (shoulderInflate_r:double), Translation="0,0,0", Rotation="0,0,0", Scale="1.2,1.2,1.2") {scaleMode=BMM_Additive, alphaBoolBlend {blendOption=Linear}, boneToModify {boneName=upperarm_twist_01_r}, scale=<1.2 1.2 1.2>}
   n40 = "Transform (Modify) Bone - Bone: upperarm_twist_01_l"(ComponentPose=@n39, Alpha=Get shoulderInflate_l() -> (shoulderInflate_l:double), Translation="0,0,0", Rotation="0,0,0", Scale="1.2,1.2,1.2") {scaleMode=BMM_Additive, alphaBoolBlend {blendOption=Linear}, boneToModify {boneName=upperarm_twist_01_l}, scale=<1.2 1.2 1.2>}
   n41 = "Constraint - dyn_jacket_back_01 (P)"(ComponentPose=@n40, Alpha=Get backpackAlpha() -> (backpackAlpha:double), ConstraintWeights_0=1) {
    alphaBoolBlend {blendOption=Linear}
    boneToModify {boneName=dyn_jacket_back_01}
    constraintSetup [{targetBone {boneName=spine_04}}]
    constraintWeights [1]
   }
   n42 = "Transform (Modify) Bone - Bone: upperCoatAdjustForHead_01_DSP"(
    ComponentPose=@n41
    Alpha=Get skydiveFloatingAlpha() -> (skydiveFloatingAlpha:double)
    Translation="-0.000,0.000,-3.957"
    Rotation="0,0,0"
    Scale="1,1,1"
   ) {
    translationMode=BMM_Additive
    translationSpace=BCS_BoneSpace
    alphaBoolBlend {blendOption=Linear}
    boneToModify {boneName=upperCoatAdjustForHead_01_DSP}
    translation {z=-3.957}
   }
   Save cached pose 'SimCache'(Pose=Component To Local(ComponentPose=@n42)) {cacheName=SimCache}
  ]}
 EventGraph {nodes [
   Event BlueprintUpdateAnimation()
   > n1 = Sequence!(execute) -> (then_0, then_1, then_2:Exec, then_3:Exec, then_4:Exec)
   n2 = GetOwnerSkeletalMeshComponent()
   n3 = BreakTransform(InTransform=GetDeltaTransformFromRefPose(self=@n2, BoneName=head, BaseName=spine_05))
   n4 = Get bIsSkyDiving()
   Set headSwingAlpha(
    execute=@n1.then_1
    headSwingAlpha:double=MapRangeClamped(Value=BreakRotator(InRot=@n3.Rotation).Yaw, InRangeA=-45, InRangeB=0, OutRangeA=SelectFloat(A=1, B=0.75, bPickA=@n4), OutRangeB=0)
   ) -> (Output_Get:double)
   n5 = GetDeltaTransformFromRefPose(self=@n2, BoneName=upperarm_r, BaseName=clavicle_r)
   n6 = "Absolute (Float)"(A=BreakRotator(InRot=BreakTransform(InTransform=@n5).Rotation).Yaw)
   > Set armSwingDelta_r(armSwingDelta_r:double=MapRangeClamped(Value=@n6, InRangeA=5, InRangeB=70)) -> (Output_Get:double)
   n7 = GetDeltaTransformFromRefPose(self=@n2, BoneName=upperarm_l, BaseName=clavicle_l)
   n8 = "Absolute (Float)"(A=BreakRotator(InRot=BreakTransform(InTransform=@n7).Rotation).Yaw)
   > Set armSwingDelta_l(armSwingDelta_l:double=MapRangeClamped(Value=@n8, InRangeA=5, InRangeB=70)) -> (Output_Get:double)
   n9 = Get bIsSkyDiving()
   > Set shoulderInflate_r(shoulderInflate_r:double=SelectFloat(A=0.5, B=MapRangeClamped(Value=@n6, InRangeA=90, InRangeB=130, OutRangeB=0.75), bPickA=@n9)) -> (Output_Get:double)
   > Set shoulderInflate_l(shoulderInflate_l:double=SelectFloat(A=0.5, B=MapRangeClamped(Value=@n8, InRangeA=10, InRangeB=80), bPickA=@n9)) -> (Output_Get:double)
   n10 = Get SkydivingExternalForceInput()
   Set windVector(
    execute=@n1.then_2
    windVector:Vector=vector * float(
     A=Break Fort Anim Input Skydiving External Force(FortAnimInput_SkydivingExternalForce=@n10) -> (
      bUseSkydivingVectorForce=~
      SkydivingVectorForce=~
      HeadToPelvisDirection=~
      FloatingMultiplier=~
      FloatingAdditive=~
      DivingMultiplier=~
      DivingAdditive=~
      ParachutingMultiplier=~
      ParachutingAdditive=~
     ).SkydivingVectorForce
     B=3
    )
   ) -> (Output_Get:Vector)
   n11 = Get CurrentLOD()
   n12 = Get bRigidBodyEnabled()
   Set rigidsOnOffBool(
    execute=@n1.then_3
    rigidsOnOffBool:bool=OR Boolean(A=NOT Boolean(A=@n12), B="integer > integer"(A=@n11, B=1) {is=title K2Node_CallFunction})
   ) -> (Output_Get:bool)
   n13 = Get bIsSkydiveFloating()
   n14 = Get bIsSkydiveDiveMode()
   n15 = Get bIsSkyDiving()
   Set skydiveFloatingAlpha(execute=@n1.then_4, skydiveFloatingAlpha:double="To Float (Boolean)"(InBool=OR Boolean!(A=@n13, B=@n15, C:bool=@n14) -> (ReturnValue))) -> (Output_Get:double)
   n16 = Get bIsJumping()
   n17 = Get bIsFalling()
   > Set jump(jump:double=SelectFloat(A=1, B=0, bPickA=OR Boolean!(A=@n16, B=@n17) -> (ReturnValue))) -> (Output_Get:double)
   n18 = Event OnCharacterCustomizationLoaded()
   n19 = TryGetPawnOwner()
   n20 = Cast To FortPlayerPawnAthena(execute=@n22.Is Valid, Object=@n19)
   Is Valid(exec=@n18.then, InputObject=@n20)
   > Do Once()
   n21 = CheckForBackbling()
   > Set backpackAlpha(backpackAlpha:double="To Float (Boolean)"(InBool=OR Boolean(A=@n21, B=@n21.bIsBackblingLong))) -> (Output_Get:double)
   Event BlueprintInitializeAnimation()
   > Set isCrouchingAlpha(isCrouchingAlpha:double=0) -> (Output_Get:double)
   > Set isNotCrouchingAlpha(isNotCrouchingAlpha:double=1) -> (Output_Get:double)
   > ResetDynamics(InTeleportType=TeleportPhysics)
   > ResetDynamics(InTeleportType=ResetPhysics)
   > n22 = Is Valid(InputObject=@n19)
   Event OnBeginCrouching()
   > Set isCrouchingAlpha(isCrouchingAlpha=1)
   > Set isNotCrouchingAlpha(isNotCrouchingAlpha=0)
   > ResetDynamics(InTeleportType=TeleportPhysics)
   > ResetDynamics(InTeleportType=ResetPhysics)
   Event OnEndCrouching()
   > Set isCrouchingAlpha(isCrouchingAlpha=0)
   > Set isNotCrouchingAlpha(isNotCrouchingAlpha=1)
   > Delay(Duration=0.125)
   > ResetDynamics(InTeleportType=TeleportPhysics)
   > ResetDynamics(InTeleportType=ResetPhysics)
   TryGetPawnOwner()
   "Not Equal (Object)"!(A:Object Reference=GetSkeletalMeshForPartType(self=@n20, PartType=Backpack), B:Object Reference) -> (ReturnValue:bool)
  ]}
]
variables [
 armSwingDelta_l {type=number}
 armSwingDelta_r {type=number}
 backpackAlpha {type=number}
 headSwingAlpha {type=number}
 isCrouchingAlpha {type=number}
 isNotCrouchingAlpha {type=number, default=1}
 jump {type=number}
 rigidsOnOffBool {type=boolean}
 shoulderInflate_l {type=number}
 shoulderInflate_r {type=number}
 skydiveFloatingAlpha {type=number}
 windVector {type=Vector}
]
