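# Physics asset for the BulletBlue female medium body's face accessory mesh, setting up secondary motion for hair strands, side locks and bun.
# Box bodies on the dyn_hair bones surround a head body built from two capsules.
# Most chains hang from tiny MSTR anchor boxes whose head joints rotate freely, with swing and twist limits on the joints below.
# Jiggle pieces constrain straight to the head, while the back lock links to both head and bun.
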PhysicsAsset /BRCosmetics/Characters/Player/Female/Medium/Bodies/F_MED_BulletBlue/Meshes/Parts/F_MED_Bullet_Blue_FaceAcc_Physics
bodies {
 dyn_hairBun_01_l [{shapeType=Box, extentX=4.3334, extentY=2.1022, extentZ=2.4994, center=<2.6661 0 0>}]
 dyn_hairback_02_l [{shapeType=Box, extentX=3.51, extentY=2.01, extentZ=1.01, center=<1.8879 0 0>}]
 dyn_hairbun_01 [{shapeType=Box, extentX=3.3334, extentY=9.1022, extentZ=10.9994, center=<1.3405 0 0>}]
 dyn_hairbun_01_MSTR [{shapeType=Box, extentX=0.01, extentY=0.01, extentZ=0.01}]
 dyn_hairbun_02 [{shapeType=Box, extentX=3.3334, extentY=9.1022, extentZ=10.9994, center=<1.3405 0 0>}]
 dyn_hairfront_01_l [{shapeType=Box, extentX=3.51, extentY=2.01, extentZ=1.01, center=<1.8879 0 0>}]
 dyn_hairfront_01_l_MSTR [{shapeType=Box, extentX=0.01, extentY=0.01, extentZ=0.01}]
 dyn_hairfrontstrand_01_l [{shapeType=Box, extentX=6.51, extentY=2.01, extentZ=1.01, center=<3.3318 0.0001 -0.0002>}]
 dyn_hairfrontstrand_01_l_MSTR [{shapeType=Box, extentX=0.01, extentY=0.01, extentZ=0.01}]
 dyn_hairfrontstrand_03_l [{shapeType=Box, extentX=6.51, extentY=2.01, extentZ=1.01, center=<3.3318 0.0001 -0.0002>}]
 dyn_hairjiggleBack_01 [{shapeType=Box, extentX=3.51, extentY=2.01, extentZ=1.01, center=<1.8879 0 0>}]
 dyn_hairjiggleFront_01 [{shapeType=Box, extentX=3.51, extentY=2.01, extentZ=1.01, center=<1.8879 0 0>}]
 dyn_hairjiggle_01_r [{shapeType=Box, extentX=3.51, extentY=2.01, extentZ=1.01, center=<1.8879 0 0>}]
 dyn_hairsideBack_01_l [{shapeType=Box, extentX=3.51, extentY=2.01, extentZ=1.01, center=<1.8879 0 0>}]
 dyn_hairsideBack_01_l_MSTR [{shapeType=Box, extentX=0.01, extentY=0.01, extentZ=0.01}]
 dyn_hairsideBack_01_r [{shapeType=Box, extentX=3.51, extentY=2.01, extentZ=1.01, center=<-1.8879 0 0>, rotation=<rot 0 0 180>}]
 dyn_hairsideBack_01_r_MSTR [{shapeType=Box, extentX=0.01, extentY=0.01, extentZ=0.01}]
 dyn_hairsideFront_01_l [{shapeType=Box, extentX=3.51, extentY=2.01, extentZ=1.01, center=<1.8879 0 0>}]
 dyn_hairsideFront_01_l_MSTR [{shapeType=Box, extentX=0.01, extentY=0.01, extentZ=0.01}]
 dyn_hairsideFront_01_r [{shapeType=Box, extentX=3.51, extentY=2.01, extentZ=1.01, center=<-1.8879 0 0>, rotation=<rot 0 0 180>}]
 dyn_hairsideFront_01_r_MSTR [{shapeType=Box, extentX=0.01, extentY=0.01, extentZ=0.01}]
 dyn_hairsideUpper_01_r [{shapeType=Box, extentX=2.51, extentY=2.01, extentZ=0.51, center=<1.2154 0.0001 -0.0002>}]
 dyn_hairsideUpper_01_r_MSTR [{shapeType=Box, extentX=0.01, extentY=0.01, extentZ=0.01}]
 dyn_hairsideUpper_02_r [{shapeType=Box, extentX=2.51, extentY=2.01, extentZ=0.51, center=<1.2154 0.0001 -0.0002>}]
 dyn_hairside_01_l [{shapeType=Box, extentX=2.51, extentY=2.01, extentZ=1.51, center=<1.2154 0.0001 -0.0002>}]
 dyn_hairside_01_l_MSTR [{shapeType=Box, extentX=0.01, extentY=0.01, extentZ=0.01}]
 dyn_hairside_01_r [{shapeType=Box, extentX=4.01, extentY=2.01, extentZ=1.51, center=<1.9969 0.0001 -0.0002>}]
 dyn_hairside_01_r_MSTR [{shapeType=Box, extentX=0.01, extentY=0.01, extentZ=0.01}]
 dyn_hairside_02_l [{shapeType=Box, extentX=2.51, extentY=2.01, extentZ=1.51, center=<1.2154 0.0001 -0.0002>}]
 dyn_hairside_02_r [{shapeType=Box, extentX=4.01, extentY=2.01, extentZ=1.51, center=<2.0671 0.0001 -0.0002>}]
 head [
  {shapeType=Capsule, length=5.457, radius=5.4075, center=<1.2694 4.5439 0>, rotation=<rot 90 1.2659 1.2658>}
  {shapeType=Capsule, length=6, radius=5.5, center=<2.0243 1.8619 -0.0001>, rotation=<rot 90 -0.0099 -0.0099>}
 ]
}
constraints [
 {bone1Name=dyn_hairsideFront_01_l, bone2Name=dyn_hairsideFront_01_l_MSTR, swing1LimitDegrees=25, swing2LimitDegrees=25, twistLimitDegrees=5}
 {bone1Name=dyn_hairsideBack_01_l, bone2Name=dyn_hairsideBack_01_l_MSTR, swing1LimitDegrees=10, swing2LimitDegrees=10, twistLimitDegrees=5}
 {bone1Name=dyn_hairfront_01_l, bone2Name=dyn_hairfront_01_l_MSTR, swing1LimitDegrees=25, swing2LimitDegrees=25, twistLimitDegrees=5}
 {
  bone1Name=dyn_hairfrontstrand_01_l
  bone2Name=dyn_hairfrontstrand_01_l_MSTR
  swing1LimitDegrees=45
  swing2LimitDegrees=45
  twistLimitDegrees=45
 }
 {bone1Name=dyn_hairside_01_l, bone2Name=dyn_hairside_01_l_MSTR, swing1LimitDegrees=10, swing2LimitDegrees=15, twistLimitDegrees=5}
 {bone1Name=dyn_hairside_02_l, bone2Name=dyn_hairside_01_l, swing1LimitDegrees=45, swing2LimitDegrees=45, twistLimitDegrees=5}
 {bone1Name=dyn_hairsideUpper_01_r, bone2Name=dyn_hairsideUpper_01_r_MSTR, swing1LimitDegrees=10, swing2LimitDegrees=10, twistLimitDegrees=5}
 {bone1Name=dyn_hairsideUpper_02_r, bone2Name=dyn_hairsideUpper_01_r, swing1LimitDegrees=45, swing2LimitDegrees=45, twistLimitDegrees=5}
 {bone1Name=dyn_hairbun_01, bone2Name=dyn_hairbun_01_MSTR, swing1LimitDegrees=10, swing2LimitDegrees=15, twistLimitDegrees=5}
 {bone1Name=dyn_hairbun_02, bone2Name=dyn_hairbun_01, swing1LimitDegrees=5, swing2LimitDegrees=10, twistLimitDegrees=5}
 {bone1Name=dyn_hairBun_01_l, bone2Name=dyn_hairbun_02, swing1LimitDegrees=30, swing2LimitDegrees=30, twistLimitDegrees=5}
 {bone1Name=dyn_hairsideFront_01_r, bone2Name=dyn_hairsideFront_01_r_MSTR, swing1LimitDegrees=25, swing2LimitDegrees=25, twistLimitDegrees=5}
 {bone1Name=dyn_hairsideBack_01_r, bone2Name=dyn_hairsideBack_01_r_MSTR, swing1LimitDegrees=25, swing2LimitDegrees=25, twistLimitDegrees=5}
 {
  bone1Name=dyn_hairsideFront_01_l_MSTR
  bone2Name=head
  swing1LimitDegrees=45
  swing1Motion=Free
  swing2LimitDegrees=45
  swing2Motion=Free
  twistLimitDegrees=45
  twistMotion=Free
 }
 {
  bone1Name=dyn_hairsideBack_01_l_MSTR
  bone2Name=head
  swing1LimitDegrees=45
  swing1Motion=Free
  swing2LimitDegrees=45
  swing2Motion=Free
  twistLimitDegrees=45
  twistMotion=Free
 }
 {
  bone1Name=dyn_hairsideFront_01_r_MSTR
  bone2Name=head
  swing1LimitDegrees=45
  swing1Motion=Free
  swing2LimitDegrees=45
  swing2Motion=Free
  twistLimitDegrees=45
  twistMotion=Free
 }
 {
  bone1Name=dyn_hairsideBack_01_r_MSTR
  bone2Name=head
  swing1LimitDegrees=45
  swing1Motion=Free
  swing2LimitDegrees=45
  swing2Motion=Free
  twistLimitDegrees=45
  twistMotion=Free
 }
 {
  bone1Name=dyn_hairfront_01_l_MSTR
  bone2Name=head
  swing1LimitDegrees=45
  swing1Motion=Free
  swing2LimitDegrees=45
  swing2Motion=Free
  twistLimitDegrees=45
  twistMotion=Free
 }
 {
  bone1Name=dyn_hairfrontstrand_01_l_MSTR
  bone2Name=head
  swing1LimitDegrees=45
  swing1Motion=Free
  swing2LimitDegrees=45
  swing2Motion=Free
  twistLimitDegrees=45
  twistMotion=Free
 }
 {
  bone1Name=dyn_hairside_01_l_MSTR
  bone2Name=head
  swing1LimitDegrees=45
  swing1Motion=Free
  swing2LimitDegrees=45
  swing2Motion=Free
  twistLimitDegrees=45
  twistMotion=Free
 }
 {
  bone1Name=dyn_hairsideUpper_01_r_MSTR
  bone2Name=head
  swing1LimitDegrees=45
  swing1Motion=Free
  swing2LimitDegrees=45
  swing2Motion=Free
  twistLimitDegrees=45
  twistMotion=Free
 }
 {
  bone1Name=dyn_hairside_01_r_MSTR
  bone2Name=head
  swing1LimitDegrees=45
  swing1Motion=Free
  swing2LimitDegrees=45
  swing2Motion=Free
  twistLimitDegrees=45
  twistMotion=Free
 }
 {
  bone1Name=dyn_hairbun_01_MSTR
  bone2Name=head
  swing1LimitDegrees=45
  swing1Motion=Free
  swing2LimitDegrees=45
  swing2Motion=Free
  twistLimitDegrees=45
  twistMotion=Free
 }
 {bone1Name=dyn_hairside_01_r, bone2Name=dyn_hairside_01_r_MSTR, swing1LimitDegrees=5, swing2LimitDegrees=10, twistLimitDegrees=5}
 {bone1Name=dyn_hairside_02_r, bone2Name=dyn_hairside_01_r, swing1LimitDegrees=45, swing2LimitDegrees=45, twistLimitDegrees=5}
 {bone1Name=dyn_hairback_02_l, bone2Name=head, swing1LimitDegrees=10, swing2LimitDegrees=10, twistLimitDegrees=5}
 {
  bone1Name=dyn_hairback_02_l
  bone2Name=dyn_hairbun_01
  swing1LimitDegrees=45
  swing1Motion=Free
  swing2LimitDegrees=45
  swing2Motion=Free
  twistLimitDegrees=45
  twistMotion=Free
 }
 {bone1Name=dyn_hairjiggleBack_01, bone2Name=head, swing1LimitDegrees=10, swing2LimitDegrees=10, twistLimitDegrees=5}
 {bone1Name=dyn_hairjiggle_01_r, bone2Name=head, swing1LimitDegrees=10, swing2LimitDegrees=10, twistLimitDegrees=5}
 {bone1Name=dyn_hairjiggleFront_01, bone2Name=head, swing1LimitDegrees=10, swing2LimitDegrees=10, twistLimitDegrees=5}
 {bone1Name=dyn_hairfrontstrand_03_l, bone2Name=dyn_hairfrontstrand_01_l, swing1LimitDegrees=45, swing2LimitDegrees=45, twistLimitDegrees=45}
]
