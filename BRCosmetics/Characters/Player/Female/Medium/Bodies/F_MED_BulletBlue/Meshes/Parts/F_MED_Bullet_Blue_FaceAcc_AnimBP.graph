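# Animation Blueprint for the face accessory part of the female Bullet Blue character, adding physics-driven secondary motion.
# It copies the attached parent's pose, then runs a RigidBody simulation in head-bone space with gravity override and damped component motion.
# Each update scales the skydiving force vector into a wind force for the simulation.
# Targeting events halve the simulation alpha while aiming and set it to full afterward.
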AnimBlueprint /BRCosmetics/Characters/Player/Female/Medium/Bodies/F_MED_BulletBlue/Meshes/Parts/F_MED_Bullet_Blue_FaceAcc_AnimBP
target_skeleton=&/F_MED_Bullet_Blue_FaceAcc_Skeleton
graphs [
 AnimGraph {nodes [
   n1 = RigidBody(
    ComponentPose=Local To Component(LocalPose=Copy Pose From Mesh() {bUseAttachedParent=true})
    Alpha=Get hairTargetingAlpha() -> (hairTargetingAlpha:double)
    ExternalForce=Get windVector() -> (windVector:Vector)
   ) {
    bClampLinearTranslationLimitToRefPose=true
    bForceDisableCollisionBetweenConstraintBodies=true
    bOverrideWorldGravity=true
    overridePhysicsAsset=&/F_MED_Bullet_Blue_FaceAcc_Physics
    simulationSpace=BaseBoneSpace
    alphaBoolBlend {blendOption=Linear}
    baseBoneRef {boneName=head}
    componentAppliedLinearAccClamp=<3000 3000 3000>
    componentLinearAccScale=<0.3 0.3 0.6>
    componentLinearVelScale=<0.3 0.3 0.6>
    overrideWorldGravity {z=-980}
   }
   Output Pose(Result=Component To Local(ComponentPose=@n1)) {layerGroup=None}
  ]}
 EventGraph {nodes [
   Event BlueprintUpdateAnimation()
   n1 = Get SkydivingExternalForceInput()
   > Set windVector(windVector:Vector=vector * float(
     A=Break Fort Anim Input Skydiving External Force(FortAnimInput_SkydivingExternalForce=@n1) -> (
      bUseSkydivingVectorForce=~
      SkydivingVectorForce=~
      HeadToPelvisDirection=~
      FloatingMultiplier=~
      FloatingAdditive=~
      DivingMultiplier=~
      DivingAdditive=~
      ParachutingMultiplier=~
      ParachutingAdditive=~
     ).SkydivingVectorForce
     B=3
    )) -> (Output_Get:Vector)
   Event OnBeginTargeting() {is=title K2Node_Event}
   > Set hairTargetingAlpha(hairTargetingAlpha:double=0.5) -> (Output_Get:double)
   Event OnEndTargeting() {is=title K2Node_Event}
   > Set hairTargetingAlpha(hairTargetingAlpha=1)
   TryGetPawnOwner()
  ]}
]
variables [
 hairTargetingAlpha {type=number}
 windVector {type=Vector}
]
